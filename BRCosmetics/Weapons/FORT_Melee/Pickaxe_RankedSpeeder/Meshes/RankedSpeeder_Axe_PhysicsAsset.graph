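# Physics asset for the RankedSpeeder axe, a two-handed melee pickaxe cosmetic.
# It holds a single capsule body on the Melee_TwoHanded_Handle bone, offset and rotated from the bone origin.
# It has no constraints, so it defines one collision shape with no jointed parts.
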PhysicsAsset /BRCosmetics/Weapons/FORT_Melee/Pickaxe_RankedSpeeder/Meshes/RankedSpeeder_Axe_PhysicsAsset
bodies {Melee_TwoHanded_Handle [{shapeType=Capsule, length=41.2639, radius=13.9135, center=<0.7361 -2.3078 32.7149>, rotation=<rot 13.195 -90 2.8773>}]}
constraints []
